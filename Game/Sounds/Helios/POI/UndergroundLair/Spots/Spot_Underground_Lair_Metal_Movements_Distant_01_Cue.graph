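# Sound cue for a distant metal-movement spot one-shot in the Helios Underground Lair point of interest.
# It picks at random between two equally weighted cave distant-metal waves, then a modulator varies pitch slightly downward and volume slightly lower on each play.
# It uses a distant-creak one-shot attenuation and the indoor ambient one-shots sound class.
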SoundCue /Game/Sounds/Helios/POI/UndergroundLair/Spots/Spot_Underground_Lair_Metal_Movements_Distant_01_Cue
root=Modulator_0
nodes [
 {id=Modulator_0, children [Random_0], properties {pitchMax=1, pitchMin=0.8, volumeMax=1, volumeMin=0.7}}
 {id=Random_0, children [WavePlayer_12, WavePlayer_13], properties {weights [1,1]}}
 {id=WavePlayer_12, properties {soundWaveAssetPtr=&/Spot_Underground_Lair_Cave_Distant_Metal_01}}
 {id=WavePlayer_13, properties {soundWaveAssetPtr=&/Spot_Underground_Lair_Cave_Distant_Metal_02}}
]
properties {
 attenuationSettings=/Game/Sounds/Attenuation/Fort_TileAmbient_OneShot_Distant_Creak_Att
 maxDistance=8200
 soundClassObject=/Game/Sounds/FortSoundClassesAndModes/Fort_Ambient_OneShots_Indoors
 volumeMultiplier=0.702
}
