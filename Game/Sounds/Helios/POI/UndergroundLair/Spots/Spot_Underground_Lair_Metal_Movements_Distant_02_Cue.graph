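# Sound cue for distant metal movement spot sounds in the Helios underground lair POI.
# A random node picks evenly between two distant cave metal waves, then a modulator varies pitch downward and volume slightly on each play.
# It plays as an indoor ambient one-shot at reduced volume, using a shared distant-creak attenuation asset.
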SoundCue /Game/Sounds/Helios/POI/UndergroundLair/Spots/Spot_Underground_Lair_Metal_Movements_Distant_02_Cue
root=Modulator_0
nodes [
 {id=Modulator_0, children [Random_0], properties {pitchMax=1, pitchMin=0.8, volumeMax=1, volumeMin=0.7}}
 {id=Random_0, children [WavePlayer_14, WavePlayer_15], properties {weights [1,1]}}
 {id=WavePlayer_14, properties {soundWaveAssetPtr=&/Spot_Underground_Lair_Cave_Distant_Metal_03}}
 {id=WavePlayer_15, properties {soundWaveAssetPtr=&/Spot_Underground_Lair_Cave_Distant_Metal_04}}
]
properties {
 attenuationSettings=/Game/Sounds/Attenuation/Fort_TileAmbient_OneShot_Distant_Creak_Att
 maxDistance=8200
 soundClassObject=/Game/Sounds/FortSoundClassesAndModes/Fort_Ambient_OneShots_Indoors
 volumeMultiplier=0.501
}
