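# Sound cue for a medium metal-stress ambient spot one-shot in the Helios Underground Lair POI.
# A Random node picks evenly among thirteen metal-stress waves named for an underground train station.
# A Modulator at the root varies each play with slightly lowered random pitch and volume.
# It plays quietly on the indoor ambient one-shot class, using distant-creak attenuation and a send to a distant reverb submix.
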SoundCue /Game/Sounds/Helios/POI/UndergroundLair/Spots/Spot_Underground_Lair_Metal_Stress_Medium_Cue
root=Modulator_0
nodes [
 {id=Modulator_0, children [Random_2], properties {pitchMax=1, pitchMin=0.8, volumeMax=1, volumeMin=0.7}}
 {
  id=Random_2
  children [
   WavePlayer_4
   WavePlayer_5
   WavePlayer_6
   WavePlayer_7
   WavePlayer_10
   WavePlayer_11
   WavePlayer_12
   WavePlayer_13
   WavePlayer_14
   WavePlayer_15
   WavePlayer_16
   WavePlayer_17
   WavePlayer_18
  ]
  properties {weights [1,1,1,1,1,1,1,1,1,1,1,1,1]}
 }
 {id=WavePlayer_10, properties {soundWaveAssetPtr=&/Spot_UndergroundTrainStation_MetalStress_05}}
 {id=WavePlayer_11, properties {soundWaveAssetPtr=&/Spot_UndergroundTrainStation_MetalStress_06}}
 {id=WavePlayer_12, properties {soundWaveAssetPtr=&/Spot_UndergroundTrainStation_MetalStress_07}}
 {id=WavePlayer_13, properties {soundWaveAssetPtr=&/Spot_UndergroundTrainStation_MetalStress_08}}
 {id=WavePlayer_14, properties {soundWaveAssetPtr=&/Spot_UndergroundTrainStation_MetalStress_09}}
 {id=WavePlayer_15, properties {soundWaveAssetPtr=&/Spot_UndergroundTrainStation_MetalStress_10}}
 {id=WavePlayer_16, properties {soundWaveAssetPtr=&/Spot_UndergroundTrainStation_MetalStress_11}}
 {id=WavePlayer_17, properties {soundWaveAssetPtr=&/Spot_UndergroundTrainStation_MetalStress_12}}
 {id=WavePlayer_18, properties {soundWaveAssetPtr=&/Spot_UndergroundTrainStation_MetalStress_13}}
 {id=WavePlayer_4, properties {soundWaveAssetPtr=&/Spot_UndergroundTrainStation_MetalStress_01}}
 {id=WavePlayer_5, properties {soundWaveAssetPtr=&/Spot_UndergroundTrainStation_MetalStress_02}}
 {id=WavePlayer_6, properties {soundWaveAssetPtr=&/Spot_UndergroundTrainStation_MetalStress_03}}
 {id=WavePlayer_7, properties {soundWaveAssetPtr=&/Spot_UndergroundTrainStation_MetalStress_04}}
]
properties {
 attenuationSettings=/Game/Sounds/Attenuation/Fort_TileAmbient_OneShot_Distant_Creak_Att
 bEnableBusSends=false
 maxDistance=8200
 soundClassObject=/Game/Sounds/FortSoundClassesAndModes/Fort_Ambient_OneShots_Indoors
 volumeMultiplier=0.1
 soundSubmixSends [{minSendLevel=0.6, soundSubmix=&Fortnite_Reverb_Distant}]
}
